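; Gameplay cue notify that shows the shield-up visual on the Large Spire in the SpireBoss content.
; On execute, it casts the target to BGA_Large_Spire and applies the shared minor spire shield material as an overlay on its budgeted skeletal mesh.
; It then gets or creates a dynamic overlay material and stamps its StartTime parameter with current game time, letting the shader time the effect from activation.
; The cue contains no logic to remove the overlay.
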(fn OnExecute
  (params (MyTarget :type Actor%20Object%20Reference) (Parameters :type Gameplay%20Cue%20Parameters%20Structure%20%28by%20ref%29))
  (bind _asbga_large_spire (Utilities|Casting|CastToBGA_Large_Spire MyTarget))
  (Rendering|Material|SetOverlayMaterial (Variables|/SpireBoss/GameplayActors/BGA_Large_Spire.BGA_Large_Spire_C|GetSkeletalMeshComponentBudgeted _asbga_large_spire) "/SpireSharedAssets/Assets/Materials/M_MinorSpire_Shield.M_MinorSpire_Shield" false 0 :!class /Script/Engine.MeshComponent)
  (bind _returnvalue (Class|FortCosmeticMaterialBlueprintLibrary|GetorCreateDynamicOverlayMaterial (Variables|/SpireBoss/GameplayActors/BGA_Large_Spire.BGA_Large_Spire_C|GetSkeletalMeshComponentBudgeted _asbga_large_spire)))
  (Rendering|Material|SetScalarParameterValue _returnvalue "StartTime" (Utilities|Time|GetGameTimeinSeconds (Variables|/SpireBoss/GameplayActors/BGA_Large_Spire.BGA_Large_Spire_C|GetSkeletalMeshComponentBudgeted _asbga_large_spire)) :!class /Script/Engine.MaterialInstanceDynamic))

(fn ConstructionScript)
