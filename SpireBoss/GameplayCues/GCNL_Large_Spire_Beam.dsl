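; Looping gameplay cue for the Major Spire's beam attack, driving its visuals, audio, sound indicator and music events.
; On start it attaches and activates the Niagara beam and times its shutdown from a scalable fire time.
; Tick polls the spire's FX interface for hit data, feeding the beam, water impact FX and audio line.
; Impact decals and camera shakes spawn only after the hit point moves a set per-axis distance.
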(event EventTick
  (params (DeltaSeconds :type Float%20%28single-precision%29))
  (bind _beamendlocation_worldspace (Variables|Default|GetBeamEndLocation_WorldSpace))
  (bind _ns_caretaker_sweepbeam (Variables|Default|GetNS_Caretaker_SweepBeam))
  (Utilities|IsValid (Utilities|ToObject(Interface) (Variables|Default|GetSpireFxInterface))
    (:Is%20Valid
      (bind _transform (FXInterface|GetBeamStopTransform (Variables|Default|GetSpireFxInterface) :!class /SpireBoss/Gameplay/BPI_MajorSpire_FXInterface.BPI_MajorSpire_FXInterface_C))
      (bind _transform_location (Math|Transform|BreakTransform _transform))
      (Variables|Default|SetBeamEndLocation_WorldSpace _transform_location)
      (bind _hitnormal (FXInterface|GetStopHitNormal (Variables|Default|GetSpireFxInterface) :!class /SpireBoss/Gameplay/BPI_MajorSpire_FXInterface.BPI_MajorSpire_FXInterface_C))
      (Variables|Default|SetBeam%20End%20Normal _hitnormal)
      (bind _actor (FXInterface|GetStopHitActor (Variables|Default|GetSpireFxInterface) :!class /SpireBoss/Gameplay/BPI_MajorSpire_FXInterface.BPI_MajorSpire_FXInterface_C))
      (Variables|Default|SetHit%20Actor _actor)
      (bind _foundhit (FXInterface|GetFoundBlockingHit (Variables|Default|GetSpireFxInterface) :!class /SpireBoss/Gameplay/BPI_MajorSpire_FXInterface.BPI_MajorSpire_FXInterface_C))
      (Variables|Default|SetBlocking%20Hit _foundhit)
      (bind _foundwaterblockinghit (FXInterface|GetFoundWaterBlockingHit (Variables|Default|GetSpireFxInterface) :!class /SpireBoss/Gameplay/BPI_MajorSpire_FXInterface.BPI_MajorSpire_FXInterface_C))
      (Variables|Default|SetisWaterBlockingHit _foundwaterblockinghit)
      (CallFunction|CurrentWaterBodyInfo)
      (CallFunction|SpawnCorrectImpactFX)
      (Effects|Components|ParticleSystem|SetActorParameter _ns_caretaker_sweepbeam "Caretaker" (Utilities|Casting|CastToActor (Utilities|ToObject(Interface) (Variables|Default|GetSpireFxInterface))))
      (Effects|Components|ParticleSystem|SetVectorParameter _ns_caretaker_sweepbeam "Stop Location" (Variables|Default|GetBeamEndLocation_WorldSpace))
      (Effects|Components|ParticleSystem|SetVectorParameter _ns_caretaker_sweepbeam "Stop Normal" (Variables|Default|GetBeam%20End%20Normal))
      (Effects|Components|ParticleSystem|SetBoolParameter _ns_caretaker_sweepbeam "isHit" (Variables|Default|GetBlocking%20Hit))
      (CallFunction|SetImpactFXValues)
      (CallFunction|DecalSpawn (Math|Boolean|ANDBoolean (Variables|Default|GetBlocking%20Hit) (not (Variables|Default|GetisWaterBlockingHit)) (not (Variables|Default|GetImpactDisabled))) _beamendlocation_worldspace)
      (CallFunction|CameraShakeSpawn (Variables|Default|GetBlocking%20Hit) _beamendlocation_worldspace)
      (CallFunction|UpdateAudioLocation))))

(event EventOnLoopingStart(EitherVFX-preferthisversion)
  (params (MyTarget :type Actor%20Object%20Reference) (Parameters :type Gameplay%20Cue%20Parameters%20Structure%20%28by%20ref%29) (ParticleComponents :type Array%20of%20FXSystem%20Component%20Object%20References) (AudioComponents :type Array%20of%20Audio%20Component%20Object%20References))
  (bind _spirefxinterface (Variables|Default|GetSpireFxInterface))
  (bind _returnvalue (Utilities|ToObject(Interface) _spirefxinterface))
  (bind _sf_beamfiretime (Variables|Default|GetSF_BeamFireTime))
  (bind _asactor (Utilities|Casting|CastToActor _returnvalue))
  (bind _returnvalue_2 (ScalableFloat|GetValueatLevelwithContextParam _asactor _sf_beamfiretime))
  (bind _ns_caretaker_sweepbeam (Variables|Default|GetNS_Caretaker_SweepBeam))
  (bind _output_get (Variables|Default|SetSpire%20Actor MyTarget))
  (Variables|Default|SetImpactDisabled false)
  (bind _asbpi_major_spire_fxinterface (Utilities|Casting|CastToBPI_MajorSpire_FXInterface _output_get))
  (Variables|Default|SetSpireFxInterface _asbpi_major_spire_fxinterface)
  (Niagara|SetNiagaraVariable(Float) (Variables|Default|GetNS_Caretaker_SweepBeam) "Duration" _returnvalue_2)
  (Niagara|SetNiagaraVariable(Bool) _ns_caretaker_sweepbeam "bHitReaction" false)
  (Components|Activation|Activate _ns_caretaker_sweepbeam false)
  (Utilities|Time|SetTimerbyEvent (EventDispatchers|CreateEvent :!event Custom|EndBeamDelay) (+ _returnvalue_2 (Variables|Default|GetBeamEndBuffer)))
  (bind _seketalmeshcomp (FXInterface|GetSkeletalMeshComponentBudgeted (Variables|Default|GetSpireFxInterface) :!class /SpireBoss/Gameplay/BPI_MajorSpire_FXInterface.BPI_MajorSpire_FXInterface_C))
  (Transformation|AttachComponentToComponent (Variables|Default|GetNS_Caretaker_SweepBeam) _seketalmeshcomp "shard_center_mask" "SnapToTarget" "SnapToTarget" "SnapToTarget" true)
  (bind _returnvalue_3 (Utilities|Time|SetTimerbyEvent (delegate Custom|ShowFuseIndicator) 0.3 true))
  (Variables|Default|SetSoundIndicatorTimer _returnvalue_3)
  (Shape|Enable (Variables|Default|GetFortAudioShapeLine))
  (CallFunction|StartBeamAudio)
  (MusicEncounter|ExecuteMusicEncounterGameplayEventByName (GameplayEventRouter|GetGlobalEventRouter (Variables|Default|GetSpire%20Actor)) "StartBeamAttack" (Variables|Default|GetSpire%20Actor) (Variables|Default|GetSpire%20Actor)))

(event Custom|ShowFuseIndicator
  (Fortnite|ShowSoundIndicatorLocalforTeam :WorldContextObject (Utilities|Casting|CastToActor (Utilities|ToObject(Interface) (Variables|Default|GetSpireFxInterface))) :TrackedActor (Utilities|Casting|CastToActor (Utilities|ToObject(Interface) (Variables|Default|GetSpireFxInterface))) :Location (Transformation|GetActorLocation (Utilities|Casting|CastToActor (Utilities|ToObject(Interface) (Variables|Default|GetSpireFxInterface)))) :MaxAudibleDistance 1.08e+04 :IndicatorType "Gunshot" :Team (Team|GetActorTeam (Utilities|Casting|CastToActor (Utilities|ToObject(Interface) (Variables|Default|GetSpireFxInterface)))) :Affiliations (Utilities|Array|MakeArray "Neutral" "Hostile") :Instigator (Utilities|Casting|CastToActor (Utilities|ToObject(Interface) (Variables|Default|GetSpireFxInterface))) :Tint "(R=0.000000,G=0.000000,B=0.000000,A=0.000000)" :AttenuationMaxRangeMultiplier 1.0 :AttenuationIntensityMultiplier 1.0))

(event Custom|EndBeamDelay
  (bind _impactfx (Variables|Default|GetImpactFX))
  (CallFunction|StopAudio)
  (Utilities|IsValid _impactfx
    (:Is%20Valid
      (Components|Activation|Deactivate _impactfx)
      (Variables|Default|SetImpactFX 0))
    (:Is%20Not%20Valid
      _))
  (Variables|Default|SetImpactDisabled true)
  (CallFunction|PlayStopBeamAudioAttached)
  (MusicEncounter|ExecuteMusicEncounterGameplayEventByName (GameplayEventRouter|GetGlobalEventRouter (Variables|Default|GetSpire%20Actor)) "StopBeamAttack" (Variables|Default|GetSpire%20Actor) (Variables|Default|GetSpire%20Actor)))

(event EventOnRemoval(EitherVFX-preferthisversion)
  (params (MyTarget :type Actor%20Object%20Reference) (Parameters :type Gameplay%20Cue%20Parameters%20Structure%20%28by%20ref%29) (ParticleComponents :type Array%20of%20FXSystem%20Component%20Object%20References) (AudioComponents :type Array%20of%20Audio%20Component%20Object%20References) (BurstCameraShakeInstance :type Camera%20Shake%20Base%20Object%20Reference) (BurstDecalInstance :type Decal%20Actor%20Object%20Reference))
  (bind _impactfx (Variables|Default|GetImpactFX))
  (Niagara|SetNiagaraVariable(Bool) (Variables|Default|GetNS_Caretaker_SweepBeam) "bHitReaction" true)
  (CallFunction|StopAudio)
  (Utilities|IsValid _impactfx
    (:Is%20Valid
      (Components|Activation|Deactivate _impactfx)
      (Variables|Default|SetImpactFX 0))
    (:Is%20Not%20Valid
      _))
  (Variables|Default|SetImpactDisabled true))

(fn CameraShakeSpawn
  (params (Blocking%20Hit :type Boolean) (Hit%20Location :type Vector))
  (if (and Blocking%20Hit (CallFunction|DistanceCheck Hit%20Location 7.5e+02 (Variables|Default|GetPrevious%20CameraShake%20Location)))
    (Camera|PlayWorldCameraShake 0 "/FNGameplayCues/Effects/Fort_Effects/Camera_Effects/BP_CameraShake_Cube_Smack_New.BP_CameraShake_Cube_Smack_New_C" Hit%20Location 0.0 2e+03 0.5)
    (Variables|Default|SetPrevious%20CameraShake%20Location Hit%20Location)))

(fn CurrentWaterBodyInfo
  (bind _hit_actor (Variables|Default|GetHit%20Actor))
  (bind (_asfort_water_body_actor _bsuccess) (Utilities|Casting|CastToFortWaterBodyActor _hit_actor))
  (if (Variables|Default|GetisWaterBlockingHit)
    (Utilities|IsValid _hit_actor
      (:Is%20Valid
        (if _bsuccess
          (bind _outwatersurfaceinfo (WaterBody|GetWaterSurfaceInfoStruct _asfort_water_body_actor (Variables|Default|GetBeamEndLocation_WorldSpace) true false false))
          (Variables|Default|SetWater%20Depth (Utilities|Struct|BreakWaterSurfaceInfo _outwatersurfaceinfo))
          (Variables|Default|SetWaterBodyIdx (Math|Conversions|ToFloat(Integer) (.WaterBodyIdx (Utilities|Struct|BreakWaterSurfaceInfo _outwatersurfaceinfo))))
          (Variables|Default|SetWave%20Attenuation (Variables|/Script/Water.WaterBodyComponent|GetTargetWaveMaskDepth (Water|GetWaterBodyComponent _asfort_water_body_actor)))
          (else
            _))
        (return))
      (:Is%20Not%20Valid
        _))
    (else
      _)))

(fn DecalSpawn
  (params (Blocking%20Hit :type Boolean) (Hit%20Location :type Vector))
  (if (and Blocking%20Hit (CallFunction|DistanceCheck Hit%20Location 1.5e+02 (Variables|Default|GetPrevious%20Decal%20Location)))
    (bind _returnvalue (Game|SpawnActorfromClass 0 "/SpireBoss/VFX/B_Decal_Spire_Beam_Impact.B_Decal_Spire_Beam_Impact_C" (Math|Transform|MakeTransform Hit%20Location) "Undefined" "MultiplyWithRoot" 0 0))
    (Rendering|Components|Decal|SetDecalColor (Variables|/Script/Engine.DecalActor|GetDecal _returnvalue) (Variables|Default|GetDecalColor))
    (Variables|Default|SetPrevious%20Decal%20Location Hit%20Location)
    (Variables|Default|GetBeamImpactAudio
      (:then)
      (:else
        (bind _returnvalue_2 (Audio|SpawnSoundatLocation 0 (Variables|Sound|GetBeamImpactSound) Hit%20Location))
        (Variables|Default|SetBeamImpactAudio _returnvalue_2)))))

(fn DistanceCheck
  (params (Hit%20Location :type Vector) (Distance%20to%20Check :type Float%20%28double-precision%29) (Previous%20Location :type Vector))
  (return (Math|Boolean|ORBoolean (> (.x Hit%20Location) (+ (.x Previous%20Location) Distance%20to%20Check)) (< (.x Hit%20Location) (+ (.x Previous%20Location) (* Distance%20to%20Check -1.0))) (> (.y Hit%20Location) (+ (.y Previous%20Location) Distance%20to%20Check)) (< (.y Hit%20Location) (+ (.y Previous%20Location) (* Distance%20to%20Check -1.0))) (> (.z Hit%20Location) (+ (.z Previous%20Location) Distance%20to%20Check)) (< (.z Hit%20Location) (+ (.z Previous%20Location) (* Distance%20to%20Check -1.0))))))

(fn PlayStopBeamAudioAttached
  (bind _seketalmeshcomp (FXInterface|GetSkeletalMeshComponentBudgeted (Variables|Default|GetSpireFxInterface) :!class /SpireBoss/Gameplay/BPI_MajorSpire_FXInterface.BPI_MajorSpire_FXInterface_C))
  (Audio|SpawnSoundAttached "/SpireBoss/Audio/Cues/SC_MajorSpire_BeamStop_Cue.SC_MajorSpire_BeamStop_Cue" _seketalmeshcomp "pillar_center"))

(fn SetImpactFXValues
  (bind _iswaterblockinghit (Variables|Default|GetisWaterBlockingHit))
  (bind _impactfx (Variables|Default|GetImpactFX))
  (Utilities|IsValid (Variables|Default|GetImpactFX)
    (:Is%20Valid
      (Effects|Components|ParticleSystem|SetVectorParameter _impactfx "Stop Location" (Variables|Default|GetBeamEndLocation_WorldSpace))
      (Effects|Components|ParticleSystem|SetVectorParameter _impactfx "Stop Normal" (Variables|Default|GetBeam%20End%20Normal))
      (Effects|Components|ParticleSystem|SetBoolParameter _impactfx "isHit" (Variables|Default|GetBlocking%20Hit))
      (if _iswaterblockinghit
        (Effects|Components|ParticleSystem|SetBoolParameter _impactfx "isWater" _iswaterblockinghit)
        (Effects|Components|ParticleSystem|SetFloatParameter _impactfx "WaterBodyIndex" (Variables|Default|GetWaterBodyIdx))
        (Effects|Components|ParticleSystem|SetFloatParameter _impactfx "WaterDepth" (Variables|Default|GetWater%20Depth))
        (Effects|Components|ParticleSystem|SetFloatParameter _impactfx "WaveAttenuationFactor" (Variables|Default|GetWave%20Attenuation))
        (Variables|Default|SetWasWaterBlocking true)
        (return)
        (else
          (Variables|Default|SetWasWaterBlocking false)
          (return))))
    (:Is%20Not%20Valid
      (return))))

(fn SpawnCorrectImpactFX
  (bind _impactfx (Variables|Default|GetImpactFX))
  (if (Variables|Default|GetImpactDisabled)
    _
    (elif (Variables|Default|GetBlocking%20Hit)
      (if (Variables|Default|GetisWaterBlockingHit)
        (Utilities|IsValid (Variables|Default|GetImpactFX)
          (:Is%20Valid
            (if (Variables|Default|GetWasWaterBlocking)
              (return)
              (else
                (Components|Activation|Deactivate (Variables|Default|GetImpactFX))
                (bind _seketalmeshcomp (FXInterface|GetSkeletalMeshComponentBudgeted (Variables|Default|GetSpireFxInterface) :!class /SpireBoss/Gameplay/BPI_MajorSpire_FXInterface.BPI_MajorSpire_FXInterface_C))
                (bind _returnvalue (Niagara|SpawnSystemAttached "/SpireBoss/VFX/NS_Spire_BifucateBeam_Impact_Water.NS_Spire_BifucateBeam_Impact_Water" _seketalmeshcomp "root" "0, 0, 0" "0, 0, 0" "KeepRelativeOffset" true)) ; Water
                (Variables|Default|SetImpactFX _returnvalue)
                (return))))
          (:Is%20Not%20Valid
            (bind _seketalmeshcomp_2 (FXInterface|GetSkeletalMeshComponentBudgeted (Variables|Default|GetSpireFxInterface) :!class /SpireBoss/Gameplay/BPI_MajorSpire_FXInterface.BPI_MajorSpire_FXInterface_C))
            (bind _returnvalue_2 (Niagara|SpawnSystemAttached "/SpireBoss/VFX/NS_Spire_BifucateBeam_Impact_Water.NS_Spire_BifucateBeam_Impact_Water" _seketalmeshcomp_2 "root" "0, 0, 0" "0, 0, 0" "KeepRelativeOffset" true)) ; Water
            (Variables|Default|SetImpactFX _returnvalue_2)
            (return)))
        (else
          (Utilities|IsValid (Variables|Default|GetImpactFX)
            (:Is%20Valid
              (if (Variables|Default|GetWasWaterBlocking)
                (Components|Activation|Deactivate (Variables|Default|GetImpactFX))
                (return)
                (else
                  (return))))
            (:Is%20Not%20Valid
              (return)))))
      (else
        (Utilities|IsValid _impactfx
          (:Is%20Valid
            (Components|Activation|Deactivate _impactfx)
            (Variables|Default|SetImpactFX 0))
          (:Is%20Not%20Valid
            _))
        (return)))))

(fn StartBeamAudio
  (bind _fortaudioshapeline (Variables|Default|GetFortAudioShapeLine))
  (Control|SetStartPoint _fortaudioshapeline (Transformation|GetWorldLocation (Variables|Transformation|GetRootComponent)))
  (Control|SetEndPoint _fortaudioshapeline (Variables|Default|GetBeamEndLocation_WorldSpace)))

(fn StopAudio
  (bind _fortaudioshapeline (Variables|Default|GetFortAudioShapeLine)
    (:then
      (Shape|Disable _fortaudioshapeline))
    (:else
      _))
  (bind _beamimpactaudio (Variables|Default|GetBeamImpactAudio)
    (:then
      (Audio|Components|Audio|FadeOut _beamimpactaudio 0.25 0.0 "Linear" :!class /Script/Engine.AudioComponent)
      (Utilities|Time|ClearandInvalidateTimerbyHandle 0 (Variables|Default|GetSoundIndicatorTimer)))
    (:else)))

(fn TestLineTraceAndSetValuesIfClosestHit
  (params (Start :type Vector) (End :type Vector) (TraceChannel :type ETraceTypeQuery%20Enum :default "TraceTypeQuery1"))
  (bind _ignoredactors (FXInterface|GetLineTraceIgnoredActors (Variables|Default|GetSpireFxInterface) :!class /SpireBoss/Gameplay/BPI_MajorSpire_FXInterface.BPI_MajorSpire_FXInterface_C))
  (bind (_outhit _returnvalue) (Collision|LineTraceByChannel 0 Start End TraceChannel false _ignoredactors "None" true "(R=1.000000,G=0.000000,B=0.000000,A=1.000000)" "(R=0.000000,G=1.000000,B=0.000000,A=1.000000)" 1e+01))
  (if _returnvalue
    (if (< (.Distance (Collision|BreakHitResult _outhit)) (Variables|Default|GetcurrentClosestDistance))
      (Variables|Default|SetBlocking%20Hit (Collision|BreakHitResult _outhit))
      (Variables|Default|SetcurrentClosestDistance (.Distance (Collision|BreakHitResult _outhit)))
      (Variables|Default|SetBeamEndLocation_WorldSpace (.ImpactPoint (Collision|BreakHitResult _outhit)))
      (Variables|Default|SetBeam%20End%20Normal (.Normal (Collision|BreakHitResult _outhit)))
      (Variables|Default|SetHit%20Actor (.HitActor (Collision|BreakHitResult _outhit)))
      (return (Collision|BreakHitResult _outhit))
      (else
        _)))
  (return false))

(fn UpdateAudioLocation
  (Control|SetStartPoint (Variables|Default|GetFortAudioShapeLine) (Transformation|GetWorldLocation (Variables|Transformation|GetRootComponent)))
  (Control|SetEndPoint (Variables|Default|GetFortAudioShapeLine) (Variables|Default|GetBeamEndLocation_WorldSpace))
  (bind _beamimpactaudio (Variables|Default|GetBeamImpactAudio)
    (:then
      (Transformation|SetWorldLocation _beamimpactaudio (Variables|/Script/AudioShapes.AudioShapeLineComponent|GetEndPoint (Variables|Default|GetFortAudioShapeLine)) false false))
    (:else)))

(fn ConstructionScript)
