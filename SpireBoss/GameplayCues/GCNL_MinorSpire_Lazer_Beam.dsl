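; Looping gameplay cue for the Minor Spire's laser beam visuals and firing sound.
; On loop start, it stores the spawned Niagara component and snaps it to the target's budgeted skeletal mesh at "masksocket".
; The cue's location and raw magnitude set the beam's stop point and duration, while a fire sound and gunshot-type sound indicator play at the target.
; On removal, it deactivates the stored beam component.
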(event EventOnLoopingStart(Niagara)
  (params (MyTarget :type Actor%20Object%20Reference) (Parameters :type Gameplay%20Cue%20Parameters%20Structure%20%28by%20ref%29) (ParticleComponents :type Array%20of%20Niagara%20Particle%20System%20Component%20Object%20References) (AudioComponents :type Array%20of%20Audio%20Component%20Object%20References))
  (bind (_normalizedmagnitude _rawmagnitude _effectcontext _matchedtagname _originaltag _aggregatedsourcetags _aggregatedtargettags _location) (Ability|GameplayCue|BreakGameplayCueParameters Parameters))
  (bind _returnvalue (Transformation|GetActorLocation MyTarget))
  (bind _niagara (Variables|Default|GetNiagara))
  (bind _output_get (Variables|Default|SetNiagara (Utilities|Array|Get(acopy) ParticleComponents)))
  (Transformation|AttachComponentToComponent _output_get (Actor|GetComponentbyClass MyTarget "/Script/AnimationBudgetAllocator.SkeletalMeshComponentBudgeted") "masksocket" "SnapToTarget" "KeepRelative" "KeepRelative" true)
  (Effects|Components|ParticleSystem|SetVectorParameter _niagara "Stop Location" _location)
  (Effects|Components|ParticleSystem|SetFloatParameter _niagara "Duration" _rawmagnitude)
  (Fortnite|ShowSoundIndicatorLocal :WorldContextObject self :TrackedActor MyTarget :Location _returnvalue :MaxAudibleDistance 5e+03 :IndicatorType "Gunshot" :Instigator MyTarget :Tint "(R=0.000000,G=0.000000,B=0.000000,A=0.000000)" :AttenuationMaxRangeMultiplier 1.0 :AttenuationIntensityMultiplier 1.0)
  (Audio|PlaySoundatLocation MyTarget "/SpireSharedAssets/Audio/Cues/SC_MinorSpire_Beam_Fire.SC_MinorSpire_Beam_Fire" _returnvalue))

(event EventOnRemoval(EitherVFX-preferthisversion)
  (params (MyTarget :type Actor%20Object%20Reference) (Parameters :type Gameplay%20Cue%20Parameters%20Structure%20%28by%20ref%29) (ParticleComponents :type Array%20of%20FXSystem%20Component%20Object%20References) (AudioComponents :type Array%20of%20Audio%20Component%20Object%20References) (BurstCameraShakeInstance :type Camera%20Shake%20Base%20Object%20Reference) (BurstDecalInstance :type Decal%20Actor%20Object%20Reference))
  (Components|Activation|Deactivate (Variables|Default|GetNiagara)))

(fn ConstructionScript)
