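; Burst gameplay cue that spawns the telegraph effect for the Spire boss's large projectile.
; After the parent burst runs, it finds the shard_center_mask socket on the instigator's budgeted skeletal mesh in actor space, adds a forward X offset and converts it to world space.
; It moves the first spawned particle component there and sets its Niagara EndLocation to the cue's location, so the effect spans from boss to target point.
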(fn OnBurst(EitherVFX-preferthisversion)
  (params (MyTarget :type Actor%20Object%20Reference) (Parameters :type Gameplay%20Cue%20Parameters%20Structure%20%28by%20ref%29) (ParticleComponents :type Array%20of%20FXSystem%20Component%20Object%20References) (AudioComponents :type Array%20of%20Audio%20Component%20Object%20References) (BurstCameraShakeInstance :type Camera%20Shake%20Base%20Object%20Reference) (BurstDecalInstance :type Decal%20Actor%20Object%20Reference))
  (bind (_normalizedmagnitude _rawmagnitude _effectcontext _matchedtagname _originaltag _aggregatedsourcetags _aggregatedtargettags _location _normal _instigator) (Ability|GameplayCue|BreakGameplayCueParameters Parameters))
  (bind _returnvalue_location (Math|Transform|BreakTransform (Transformation|GetSocketTransform (Actor|GetComponentbyClass _instigator "/Script/AnimationBudgetAllocator.SkeletalMeshComponentBudgeted") "shard_center_mask" "RTS_Actor")))
  (bind _output (Utilities|Array|Get(acopy) ParticleComponents))
  (|CallParentFunction MyTarget Parameters ParticleComponents AudioComponents BurstCameraShakeInstance BurstDecalInstance)
  (Transformation|SetWorldLocation _output (Math|Transform|TransformLocation (Transformation|GetActorTransform _instigator) (+ _returnvalue_location "100.000000,0.000000,0.000000")) false false)
  (bind _asniagara_particle_system_component (Utilities|Casting|CastToNiagaraComponent _output))
  (Niagara|SetNiagaraVariable(Vector3) _asniagara_particle_system_component "EndLocation" _location))
