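; Burst gameplay cue for the large Spire boss's summon roar that drives its VFX from nearby minion spawn prefabs.
; It lowers the cue location by a fixed vertical offset and passes that point to the first particle component as GroundHeight.
; A sphere overlap there gathers unique locations of CherrySmoke minion encounter prefabs that allow VFX to spawn.
; Those locations and their count feed the second, Niagara component as the spike position array and spike amount.
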(fn OnBurst(EitherVFX-preferthisversion)
  (params (MyTarget :type Actor%20Object%20Reference) (Parameters :type Gameplay%20Cue%20Parameters%20Structure%20%28by%20ref%29) (ParticleComponents :type Array%20of%20FXSystem%20Component%20Object%20References) (AudioComponents :type Array%20of%20Audio%20Component%20Object%20References) (BurstCameraShakeInstance :type Camera%20Shake%20Base%20Object%20Reference) (BurstDecalInstance :type Decal%20Actor%20Object%20Reference))
  (bind (_normalizedmagnitude _rawmagnitude _effectcontext _matchedtagname _originaltag _aggregatedsourcetags _aggregatedtargettags _location) (Ability|GameplayCue|BreakGameplayCueParameters Parameters))
  (bind _output (Utilities|Array|Get(acopy) ParticleComponents 1))
  (bind _asniagara_particle_system_component (Utilities|Casting|CastToNiagaraComponent _output))
  (bind _returnvalue (- _location "0.000000,0.000000,600.000000"))
  (bind _positionarray (Variables|Default|GetPositionArray))
  (Utilities|FlowControl|Sequence
    (:then_0
      (Effects|Components|ParticleSystem|SetVectorParameter (Utilities|Array|Get(acopy) ParticleComponents) "GroundHeight" _returnvalue))
    (:then_1
      (bind _outactors (Collision|SphereOverlapActors 0 _returnvalue 2.4e+03 (Utilities|Array|MakeArray "ObjectTypeQuery2" "ObjectTypeQuery3" "ObjectTypeQuery6" "ObjectTypeQuery7" "ObjectTypeQuery1") "/SpireSharedAssets/LWPrefab/LWP_CherrySmoke_Minion_Base.LWP_CherrySmoke_Minion_Base_C"))
      (Utilities|Array|Clear (Variables|Default|GetPositionArray))
      (for _array_element _outactors
        (bind _aslwp_cherry_smoke_minion_encounter_base (Utilities|Casting|CastToLWP_CHerrySmoke_Minion_Encounter_Base _array_element))
        (if (Variables|/Lager/LivingWorldPrefabs/LWP_RiftIn_Base.LWP_RiftIn_Base_C|GetAllowVFXToSpawn _aslwp_cherry_smoke_minion_encounter_base)
          (Utilities|Array|AddUnique (Variables|Default|GetPositionArray) (Transformation|GetActorLocation _aslwp_cherry_smoke_minion_encounter_base))))
      (Niagara|NiagaraSetVectorArray _asniagara_particle_system_component "SpikePositionArray" _positionarray)
      (Niagara|SetNiagaraVariable(Int32) _asniagara_particle_system_component "SpikeAmount" (Utilities|Array|Length _positionarray)))
    (:then_2
      (return))))
